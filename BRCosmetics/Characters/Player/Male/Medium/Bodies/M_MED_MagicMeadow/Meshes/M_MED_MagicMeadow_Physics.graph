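# Physics asset for the MagicMeadow male medium player body cosmetic, covering the left and right hip-tie joints, pelvis and thighs with capsules.
# Each hip tie is a two-body chain, with the first joint limited in swing and twist against the pelvis and the second locking twist against the first.
# Thigh twist bones carry large capsules, likely as collision volumes for the ties, and connect to the pelvis with free angular motion.
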PhysicsAsset /BRCosmetics/Characters/Player/Male/Medium/Bodies/M_MED_MagicMeadow/Meshes/M_MED_MagicMeadow_Physics
bodies {
 dyn_L_hiptie_3_jnt [{shapeType=Capsule, length=5, radius=1.5, shapeName=dyn_L_hiptie_3_jnt_capsule, center=<3.5411 0 0>, rotation=<rot 90 0 0>}]
 dyn_L_hiptie_3_jnt_B [{shapeType=Capsule, length=3, radius=1.8, shapeName=dyn_L_hiptie_3_jnt_capsule, center=<3.5411 0 0>, rotation=<rot 90 0 0>}]
 dyn_R_hiptie_3_jnt [{shapeType=Capsule, length=5, radius=1.5, shapeName=dyn_R_hiptie_3_jnt_capsule, center=<3.587 0 0>, rotation=<rot 90 0 0>}]
 dyn_R_hiptie_3_jnt_B [{shapeType=Capsule, length=3, radius=2.1, shapeName=dyn_R_hiptie_3_jnt_capsule, center=<3.587 0 0>, rotation=<rot 90 0 0>}]
 pelvis [{shapeType=Capsule, length=11, radius=10, shapeName=pelvis_capsule, center=<-12.3581 -1.4408 0>}]
 thigh_twist_01_l [{shapeType=Capsule, length=30, radius=10, shapeName=thigh_l_capsule, center=<-1.9961 -0.0016 2.6654>, rotation=<rot 87.8382 -0.2784 -0.2787>}]
 thigh_twist_01_r [{shapeType=Capsule, length=30, radius=10, shapeName=thigh_r_capsule, center=<1.9535 0.0017 -3.0972>, rotation=<rot 87.106 -0.2782 -0.2785>}]
}
constraints [
 {bone1Name=dyn_L_hiptie_3_jnt, bone2Name=pelvis, swing1LimitDegrees=60, swing2LimitDegrees=60, twistLimitDegrees=20}
 {bone1Name=dyn_R_hiptie_3_jnt, bone2Name=pelvis, swing1LimitDegrees=60, swing2LimitDegrees=60, twistLimitDegrees=20}
 {
  bone1Name=thigh_twist_01_r
  bone2Name=pelvis
  swing1LimitDegrees=45
  swing1Motion=Free
  swing2LimitDegrees=45
  swing2Motion=Free
  twistLimitDegrees=45
  twistMotion=Free
 }
 {
  bone1Name=thigh_twist_01_l
  bone2Name=pelvis
  swing1LimitDegrees=45
  swing1Motion=Free
  swing2LimitDegrees=45
  swing2Motion=Free
  twistLimitDegrees=45
  twistMotion=Free
 }
 {
  bone1Name=dyn_R_hiptie_3_jnt_B
  bone2Name=dyn_R_hiptie_3_jnt
  swing1LimitDegrees=45
  swing2LimitDegrees=45
  twistLimitDegrees=45
  twistMotion=Locked
 }
 {
  bone1Name=dyn_L_hiptie_3_jnt_B
  bone2Name=dyn_L_hiptie_3_jnt
  swing1LimitDegrees=45
  swing2LimitDegrees=45
  twistLimitDegrees=45
  twistMotion=Locked
 }
]
