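# Animation Blueprint for the MagicMeadow male medium cosmetic body, adding secondary motion to its accessories.
# It copies the attached parent's pose through a deform Control Rig, then chains weighted bone constraints, AnimDynamics and RigidBody simulation for buckles, amulet, hip hook and ties.
# EventGraph-driven alphas blend corrective offsets: left-thigh yaw for the belt buckle, acceleration/jump/skydive for the hip hook, and front-end idle montage for the amulet.
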AnimBlueprint /BRCosmetics/Characters/Player/Male/Medium/Bodies/M_MED_MagicMeadow/Meshes/M_MED_MagicMeadow_AnimBP
target_skeleton=&/M_MED_MagicMeadow_Skeleton
graphs [
 AnimGraph {nodes [
   n1 = Control Rig(Source=Copy Pose From Mesh() {bUseAttachedParent=true}) {
    alphaBoolBlend {blendOption=Linear}
    controlRigAssetReference {blueprintRigClass=&/M_MED_DeformRig_MagicMeadow_ControlRig.M_MED_DeformRig_MagicMeadow_ControlRig_C}
   }
   n2 = "Constraint - neck_buckle_jnt (PPPPPPP)"!(
    ComponentPose=Local To Component(LocalPose=@n1)
    Alpha=1
    ConstraintWeights_0=0.0348
    ConstraintWeights_1:float=0.1832
    ConstraintWeights_2:float=0.2186
    ConstraintWeights_3:float=0.4754
    ConstraintWeights_4:float=0.0421
    ConstraintWeights_5:float=0.0225
    ConstraintWeights_6:float=0.0234
   ) -> (Pose) {
    alphaBoolBlend {blendOption=Linear}
    boneToModify {boneName=neck_buckle_jnt}
    constraintSetup [
     {targetBone {boneName=spine_03}}
     {targetBone {boneName=spine_04}}
     {targetBone {boneName=spine_05}}
     {targetBone {boneName=neck_01}}
     {targetBone {boneName=neck_02}}
     {targetBone {boneName=clavicle_l}}
     {targetBone {boneName=clavicle_r}}
    ]
    constraintWeights [1,1,1,1,1,1,1]
   }
   n3 = "Constraint - belt_buckle_jnt (PPPPPPPP)"!(
    ComponentPose=@n2
    Alpha=0.5
    ConstraintWeights_0=0.1182
    ConstraintWeights_1:float=0.0502
    ConstraintWeights_2:float=0.0343
    ConstraintWeights_3:float=0.0659
    ConstraintWeights_4:float=0.1353
    ConstraintWeights_5:float=0.0668
    ConstraintWeights_6:float=0.4715
    ConstraintWeights_7:float=0.0577
   ) -> (Pose) {
    alphaBoolBlend {blendOption=Linear}
    boneToModify {boneName=belt_buckle_jnt}
    constraintSetup [
     {targetBone {boneName=spine_01}}
     {targetBone {boneName=thigh_twist_01_r}}
     {targetBone {boneName=spine_03}}
     {targetBone {boneName=thigh_twist_01_l}}
     {targetBone {boneName=deform_groin_r}}
     {targetBone {boneName=spine_02}}
     {targetBone {boneName=pelvis}}
     {targetBone {boneName=deform_groin_l}}
    ]
    constraintWeights [1,1,1,1,1,1,1,1]
   }
   n4 = Get L_thighForward() -> (L_thighForward:double)
   n5 = "Transform (Modify) Bone - Bone: belt_buckle_secondary_jnt"!(ComponentPose=@n3, Alpha=@n4, Translation="0,0,0", Rotation="30,20,0") -> (Pose) {
    rotationMode=BMM_Additive
    alphaBoolBlend {blendOption=Linear}
    alphaScaleBiasClamp {bInterpResult=true}
    boneToModify {boneName=belt_buckle_secondary_jnt}
    rotation {pitch=30, yaw=20}
   }
   n6 = "Transform (Modify) Bone - Bone: belt_buckle_jnt"!(ComponentPose=@n5, Alpha=@n4, Translation="0,2,0", Rotation="18,4.5,0") -> (Pose) {
    rotationMode=BMM_Additive
    translationMode=BMM_Additive
    alphaBoolBlend {blendOption=Linear}
    alphaScaleBiasClamp {bInterpResult=true}
    boneToModify {boneName=belt_buckle_jnt}
    rotation {pitch=18, yaw=4.5}
    translation {y=2}
   }
   n7 = Anim Dynamics - Bone: dyn_amulet_motion_jnt(ComponentPose=@n6, Alpha=1, ExternalForce="0,0,0") {
    angularDampingOverride=0.9999
    angularSpringConstant=50
    bAngularSpring=true
    bOverrideAngularDamping=true
    simulationSpace=RootRelative
    alphaBoolBlend {blendOption=Linear}
    boundBone {boneName=dyn_amulet_motion_jnt}
    physicsBodyDefinitions [{
      boundBone {boneName=dyn_amulet_motion_jnt}
      boxExtents=<10 10 10>
      constraintSetup {angularLimitsMax=<0 7 50>, angularLimitsMin=<0 -7 5>, angularTarget=<1 0 0>}
      localJointOffset=<1 0 0>
     }]
   }
   n8 = Anim Dynamics - Bone: dyn_amulet_jnt(ComponentPose=@n7, Alpha=1, ExternalForce="0,0,0") {
    angularDampingOverride=0.9999
    angularSpringConstant=50
    bAngularSpring=true
    bOverrideAngularDamping=true
    simulationSpace=RootRelative
    alphaBoolBlend {blendOption=Linear}
    boundBone {boneName=dyn_amulet_jnt}
    physicsBodyDefinitions [{
      boundBone {boneName=dyn_amulet_jnt}
      boxExtents=<10 10 10>
      constraintSetup {angularLimitsMax=<0 7 50>, angularLimitsMin=<0 -7 5>, angularTarget=<1 0 0>}
      localJointOffset=<1 0 0>
     }]
   }
   n9 = Get customIdlePlaying() -> (customIdlePlaying:bool)
   n10 = "Transform (Modify) Bone - Bone: dyn_amulet_motion_jnt"!(ComponentPose=@n8, bAlphaBoolEnabled:bool=@n9, Translation="0,-0.5,0.3", Rotation="0,0,-5", Scale="1,1,1") -> (Pose) {
    alphaInputType=Bool
    rotationMode=BMM_Additive
    translationMode=BMM_Additive
    alphaBoolBlend {blendOption=Linear}
    boneToModify {boneName=dyn_amulet_motion_jnt}
    rotation {roll=-5}
    translation {y=-0.5, z=0.3}
   }
   n11 = "Transform (Modify) Bone - Bone: dyn_amulet_jnt"!(ComponentPose=@n10, bAlphaBoolEnabled:bool=@n9, Translation="0,0,0", Rotation="0,0,-10", Scale="1,1,1") -> (Pose) {
    alphaInputType=Bool
    rotationMode=BMM_Additive
    translationMode=BMM_Additive
    alphaBoolBlend {blendOption=Linear}
    boneToModify {boneName=dyn_amulet_jnt}
    rotation {roll=-10}
   }
   n12 = RigidBody(ComponentPose=@n11) {
    bClampLinearTranslationLimitToRefPose=true
    bOverrideWorldGravity=true
    overridePhysicsAsset=&/M_MED_MagicMeadow_Physics
    simulationSpace=BaseBoneSpace
    alphaBoolBlend {blendOption=Linear}
    baseBoneRef {boneName=spine_05}
    componentAppliedLinearAccClamp=<3000 3000 3000>
    componentLinearAccScale=<0.3 0.3 0.2>
    componentLinearVelScale=<0.3 0.3 0.2>
    overrideWorldGravity {z=-980}
    simSpaceSettings {maxAngularAcceleration=1000, maxAngularVelocity=3, maxLinearAcceleration=250, maxLinearVelocity=500, velocityScaleZ=0, worldAlpha=1}
   }
   n13 = "Transform (Modify) Bone - Bone: master_hip_hook_deform_jnt"(ComponentPose=@n12, Alpha=Get isNotMoving() -> (isNotMoving:double), Translation="-0.181,-1.341,0.366", Rotation="0,0,0", Scale="1,1,1") {
    translationMode=BMM_Additive
    translationSpace=BCS_BoneSpace
    alphaBoolBlend {blendOption=Linear}
    alphaScaleBiasClamp {bInterpResult=true}
    boneToModify {boneName=master_hip_hook_deform_jnt}
    translation=<-0.181 -1.341 0.366>
   }
   n14 = "Transform (Modify) Bone - Bone: master_hip_hook_deform_jnt"(ComponentPose=@n13, Alpha=Get isMoving() -> (isMoving:float), Translation="1.198,-6.250,1.318", Rotation="0,0,0", Scale="1,1,1") {
    translationMode=BMM_Additive
    translationSpace=BCS_BoneSpace
    alphaBoolBlend {blendOption=Linear}
    alphaScaleBiasClamp {bInterpResult=true}
    boneToModify {boneName=master_hip_hook_deform_jnt}
    translation=<1.198 -6.25 1.318>
   }
   n15 = "Constraint - master_hip_hook_deform_jnt (RRR)"!(ComponentPose=@n14, Alpha=1, ConstraintWeights_0=0.5, ConstraintWeights_1:float=0.5, ConstraintWeights_2:float=0) -> (Pose) {
    alphaBoolBlend {blendOption=Linear}
    boneToModify {boneName=master_hip_hook_deform_jnt}
    constraintSetup [
     {transformType=Rotation, targetBone {boneName=thigh_r}}
     {transformType=Rotation, targetBone {boneName=thigh_twist_01_r}}
     {transformType=Rotation, targetBone {boneName=pelvis}}
    ]
    constraintWeights [1,1,1]
   }
   n16 = "Transform (Modify) Bone - Bone: master_hip_hook_deform_jnt"(ComponentPose=@n15, Alpha=Get isMoving(), Translation="0,0,0", Rotation="0,-23,0", Scale="1,1,1") {
    rotationMode=BMM_Additive
    rotationSpace=BCS_BoneSpace
    alphaBoolBlend {blendOption=Linear}
    alphaScaleBiasClamp {bInterpResult=true}
    boneToModify {boneName=master_hip_hook_deform_jnt}
    rotation {yaw=-23}
   }
   n17 = "Constraint - dyn_L_hiptie_1_jnt (PPPPP)"!(
    ComponentPose=@n16
    Alpha=1
    ConstraintWeights_0=0.3099
    ConstraintWeights_1:float=0.0215
    ConstraintWeights_2:float=0.4897
    ConstraintWeights_3:float=0.0861
    ConstraintWeights_4:float=0.0928
   ) -> (Pose) {
    alphaBoolBlend {blendOption=Linear}
    boneToModify {boneName=dyn_L_hiptie_1_jnt}
    constraintSetup [
     {targetBone {boneName=pelvis}}
     {targetBone {boneName=deform_groin_r}}
     {targetBone {boneName=thigh_twist_01_l}}
     {targetBone {boneName=thigh_l}}
     {targetBone {boneName=deform_groin_l}}
    ]
    constraintWeights [1,1,1,1,1]
   }
   n18 = "Constraint - dyn_R_hiptie_1_jnt (PPPPPP)"!(
    ComponentPose=@n17
    Alpha=1
    ConstraintWeights_0=0.0578
    ConstraintWeights_1:float=0.238
    ConstraintWeights_2:float=0.183
    ConstraintWeights_3:float=0.5034
    ConstraintWeights_4:float=0.0104
    ConstraintWeights_5:float=0.0075
   ) -> (Pose) {
    alphaBoolBlend {blendOption=Linear}
    boneToModify {boneName=dyn_R_hiptie_1_jnt}
    constraintSetup [
     {targetBone {boneName=thigh_r}}
     {targetBone {boneName=pelvis}}
     {targetBone {boneName=deform_groin_r}}
     {targetBone {boneName=thigh_twist_01_r}}
     {targetBone {boneName=thigh_twist_01_l}}
     {targetBone {boneName=deform_groin_l}}
    ]
    constraintWeights [1,1,1,1,1,1]
   }
   n19 = Anim Dynamics - Bone: dyn_L_hiptie_1_jnt(ComponentPose=@n18, Alpha=1, ExternalForce="0,0,0") {
    angularDampingOverride=1
    bOverrideAngularDamping=true
    bUseGravityOverride=true
    simulationSpace=RootRelative
    alphaBoolBlend {blendOption=Linear}
    boundBone {boneName=dyn_L_hiptie_1_jnt}
    componentAppliedLinearAccClamp=<3000 3000 300>
    componentLinearAccScale=<0.1 0.05 0.05>
    componentLinearVelScale=<0.1 0.05 0.05>
    gravityOverride {z=-980}
    physicsBodyDefinitions [{
      boundBone {boneName=dyn_L_hiptie_1_jnt}
      boxExtents=<10 10 10>
      constraintSetup {angularLimitsMax=<10 0 90>, angularLimitsMin=<-10 -45 -90>, angularTarget=<1 0 0>}
      localJointOffset=<5 0 0>
     }]
   }
   n20 = Anim Dynamics - Bone: dyn_R_hiptie_1_jnt(ComponentPose=@n19, Alpha=1, ExternalForce="0,0,0") {
    angularDampingOverride=1
    bOverrideAngularDamping=true
    bUseGravityOverride=true
    simulationSpace=RootRelative
    alphaBoolBlend {blendOption=Linear}
    boundBone {boneName=dyn_R_hiptie_1_jnt}
    componentAppliedLinearAccClamp=<3000 3000 300>
    componentLinearAccScale=<0.3 0.2 0.2>
    componentLinearVelScale=<0.3 0.2 0.2>
    gravityOverride {z=-980}
    physicsBodyDefinitions [{
      boundBone {boneName=dyn_R_hiptie_1_jnt}
      boxExtents=<10 10 10>
      constraintSetup {angularLimitsMax=<10 45 90>, angularLimitsMin=<-10 0 -90>, angularTarget=<1 0 0>}
      localJointOffset=<5 0 0>
     }]
   }
   n21 = "Constraint - dyn_L_hiptie_2_jnt (PPPPP)"!(
    ComponentPose=@n20
    Alpha=1
    ConstraintWeights_0=0.3747
    ConstraintWeights_1:float=0.0034
    ConstraintWeights_2:float=0.5008
    ConstraintWeights_3:float=0.0625
    ConstraintWeights_4:float=0.0587
   ) -> (Pose) {
    alphaBoolBlend {blendOption=Linear}
    boneToModify {boneName=dyn_L_hiptie_2_jnt}
    constraintSetup [
     {targetBone {boneName=pelvis}}
     {targetBone {boneName=deform_groin_r}}
     {targetBone {boneName=thigh_twist_01_l}}
     {targetBone {boneName=thigh_l}}
     {targetBone {boneName=deform_groin_l}}
    ]
    constraintWeights [1,1,1,1,1]
   }
   n22 = "Constraint - dyn_R_hiptie_2_jnt (PPPPP)"!(
    ComponentPose=@n21
    Alpha=1
    ConstraintWeights_0=0.0366
    ConstraintWeights_1:float=0.3111
    ConstraintWeights_2:float=0.1112
    ConstraintWeights_3:float=0.5382
    ConstraintWeights_4:float=0.003
   ) -> (Pose) {
    alphaBoolBlend {blendOption=Linear}
    boneToModify {boneName=dyn_R_hiptie_2_jnt}
    constraintSetup [
     {targetBone {boneName=thigh_r}}
     {targetBone {boneName=pelvis}}
     {targetBone {boneName=deform_groin_r}}
     {targetBone {boneName=thigh_twist_01_r}}
     {targetBone {boneName=thigh_twist_01_l}}
    ]
    constraintWeights [1,1,1,1,1]
   }
   n23 = Anim Dynamics - Bone: dyn_L_hiptie_2_jnt(ComponentPose=@n22, Alpha=1, ExternalForce="0,0,0") {
    angularDampingOverride=1
    bOverrideAngularDamping=true
    bUseGravityOverride=true
    simulationSpace=RootRelative
    alphaBoolBlend {blendOption=Linear}
    boundBone {boneName=dyn_L_hiptie_2_jnt}
    componentAppliedLinearAccClamp=<3000 3000 300>
    componentLinearAccScale=<0.1 0.05 0.05>
    componentLinearVelScale=<0.1 0.05 0.05>
    gravityOverride {z=-980}
    physicsBodyDefinitions [{
      boundBone {boneName=dyn_L_hiptie_2_jnt}
      boxExtents=<10 10 10>
      constraintSetup {angularLimitsMax=<10 0 90>, angularLimitsMin=<-10 -45 -90>, angularTarget=<1 0 0>}
      localJointOffset=<5 0 0>
     }]
   }
   n24 = Anim Dynamics - Bone: dyn_L_hiptie_2_jnt_B(ComponentPose=@n23, Alpha=1, ExternalForce="0,0,0") {
    angularDampingOverride=1
    bOverrideAngularDamping=true
    bUseGravityOverride=true
    simulationSpace=RootRelative
    alphaBoolBlend {blendOption=Linear}
    boundBone {boneName=dyn_L_hiptie_2_jnt_B}
    componentAppliedLinearAccClamp=<3000 3000 300>
    componentLinearAccScale=<0.1 0.05 0.05>
    componentLinearVelScale=<0.1 0.05 0.05>
    gravityOverride {z=-980}
    physicsBodyDefinitions [{
      boundBone {boneName=dyn_L_hiptie_2_jnt_B}
      boxExtents=<10 10 10>
      constraintSetup {angularLimitsMax=<10 0 90>, angularLimitsMin=<-10 -45 -90>, angularTarget=<1 0 0>}
      localJointOffset=<5 0 0>
     }]
   }
   n25 = Anim Dynamics - Bone: dyn_R_hiptie_2_jnt(ComponentPose=@n24, Alpha=1, ExternalForce="0,0,0") {
    angularDampingOverride=1
    bOverrideAngularDamping=true
    bUseGravityOverride=true
    simulationSpace=RootRelative
    alphaBoolBlend {blendOption=Linear}
    boundBone {boneName=dyn_R_hiptie_2_jnt}
    componentAppliedLinearAccClamp=<3000 3000 300>
    componentLinearAccScale=<0.3 0.2 0.2>
    componentLinearVelScale=<0.3 0.2 0.2>
    gravityOverride {z=-980}
    physicsBodyDefinitions [{
      boundBone {boneName=dyn_R_hiptie_2_jnt}
      boxExtents=<10 10 10>
      constraintSetup {angularLimitsMax=<10 45 90>, angularLimitsMin=<-10 0 -90>, angularTarget=<1 0 0>}
      localJointOffset=<5 0 0>
     }]
   }
   n26 = Anim Dynamics - Bone: dyn_R_hiptie_2_jnt_B(ComponentPose=@n25, Alpha=1, ExternalForce="0,0,0") {
    angularDampingOverride=1
    bOverrideAngularDamping=true
    bUseGravityOverride=true
    simulationSpace=RootRelative
    alphaBoolBlend {blendOption=Linear}
    boundBone {boneName=dyn_R_hiptie_2_jnt_B}
    componentAppliedLinearAccClamp=<3000 3000 300>
    componentLinearAccScale=<0.3 0.2 0.2>
    componentLinearVelScale=<0.3 0.2 0.2>
    gravityOverride {z=-980}
    physicsBodyDefinitions [{
      boundBone {boneName=dyn_R_hiptie_2_jnt_B}
      boxExtents=<10 10 10>
      constraintSetup {angularLimitsMax=<10 45 90>, angularLimitsMin=<-10 0 -90>, angularTarget=<1 0 0>}
      localJointOffset=<5 0 0>
     }]
   }
   n27 = "Constraint - dyn_L_armguard_tie_1_jnt (PPPPPP)"!(
    ComponentPose=@n26
    Alpha=1
    ConstraintWeights_0=0.1416
    ConstraintWeights_1:float=0.015
    ConstraintWeights_2:float=0.5184
    ConstraintWeights_3:float=0.123
    ConstraintWeights_4:float=0.1218
    ConstraintWeights_5:float=0.0802
   ) -> (Pose) {
    alphaBoolBlend {blendOption=Linear}
    boneToModify {boneName=dyn_L_armguard_tie_1_jnt}
    constraintSetup [
     {targetBone {boneName=lowerarm_l}}
     {targetBone {boneName=upperarm_twist_02_l}}
     {targetBone {boneName=deform_elbow_out_l}}
     {targetBone {boneName=upperarm_l}}
     {targetBone {boneName=lowerarm_twist_02_l}}
     {targetBone {boneName=dyn_L_armguard_tie_1_jnt}}
    ]
    constraintWeights [1,1,1,1,1,1]
   }
   n28 = "Constraint - dyn_R_armguard_tie_1_jnt (PPPPPPP)"!(
    ComponentPose=@n27
    Alpha=1
    ConstraintWeights_0=0.5574
    ConstraintWeights_1:float=0.2101
    ConstraintWeights_2:float=0.0428
    ConstraintWeights_3:float=0.0918
    ConstraintWeights_4:float=0.0669
    ConstraintWeights_5:float=0.0187
    ConstraintWeights_6:float=0.0122
   ) -> (Pose) {
    alphaBoolBlend {blendOption=Linear}
    boneToModify {boneName=dyn_R_armguard_tie_1_jnt}
    constraintSetup [
     {targetBone {boneName=deform_elbow_out_r}}
     {targetBone {boneName=dyn_R_armguard_tie_1_jnt}}
     {targetBone {boneName=lowerarm_twist_02_r}}
     {targetBone {boneName=upperarm_r}}
     {targetBone {boneName=lowerarm_r}}
     {targetBone {boneName=deform_elbow_in_r}}
     {targetBone {boneName=upperarm_twist_02_r}}
    ]
    constraintWeights [1,1,1,1,1,1,1]
   }
   n29 = Anim Dynamics - Bone: dyn_L_armguard_tie_1_jnt(ComponentPose=@n28, Alpha=1, ExternalForce="0,0,0") {
    angularDampingOverride=0.9999
    angularSpringConstant=100
    bAngularSpring=true
    bOverrideAngularDamping=true
    bUseGravityOverride=true
    simulationSpace=RootRelative
    alphaBoolBlend {blendOption=Linear}
    boundBone {boneName=dyn_L_armguard_tie_1_jnt}
    componentAppliedLinearAccClamp=<3000 3000 300>
    gravityOverride {z=-980}
    physicsBodyDefinitions [{
      boundBone {boneName=dyn_L_armguard_tie_1_jnt}
      boxExtents=<5 5 5>
      constraintSetup {angularLimitsMax=<10 30 5>, angularLimitsMin=<-10 -30 -100>, angularTarget=<1 0 0>}
      localJointOffset=<-1 0 0>
     }]
   }
   n30 = Anim Dynamics - Bone: dyn_L_armguard_tie_2_jnt(ComponentPose=@n29, Alpha=1, ExternalForce="0,0,0") {
    angularDampingOverride=0.9999
    angularSpringConstant=100
    bAngularSpring=true
    bOverrideAngularDamping=true
    bUseGravityOverride=true
    simulationSpace=RootRelative
    alphaBoolBlend {blendOption=Linear}
    boundBone {boneName=dyn_L_armguard_tie_2_jnt}
    componentAppliedLinearAccClamp=<3000 3000 300>
    gravityOverride {z=-980}
    physicsBodyDefinitions [{
      boundBone {boneName=dyn_L_armguard_tie_2_jnt}
      boxExtents=<5 5 5>
      constraintSetup {angularLimitsMax=<10 60 5>, angularLimitsMin=<-10 -60 -100>, angularTarget=<1 0 0>}
      localJointOffset=<-1 0 0>
     }]
   }
   n31 = Anim Dynamics - Bone: dyn_R_armguard_tie_1_jnt(ComponentPose=@n30, Alpha=1, ExternalForce="0,0,0") {
    angularDampingOverride=0.9999
    angularSpringConstant=100
    bAngularSpring=true
    bOverrideAngularDamping=true
    bUseGravityOverride=true
    simulationSpace=RootRelative
    alphaBoolBlend {blendOption=Linear}
    boundBone {boneName=dyn_R_armguard_tie_1_jnt}
    componentAppliedLinearAccClamp=<3000 3000 300>
    gravityOverride {z=-980}
    physicsBodyDefinitions [{
      boundBone {boneName=dyn_R_armguard_tie_1_jnt}
      boxExtents=<5 5 5>
      constraintSetup {angularLimitsMax=<10 30 5>, angularLimitsMin=<-10 -30 -100>, angularTarget=<1 0 0>}
      localJointOffset=<1 0 0>
     }]
   }
   n32 = Anim Dynamics - Bone: dyn_R_armguard_tie_2_jnt(ComponentPose=@n31, Alpha=1, ExternalForce="0,0,0") {
    angularDampingOverride=0.9999
    angularSpringConstant=100
    bAngularSpring=true
    bOverrideAngularDamping=true
    bUseGravityOverride=true
    simulationSpace=RootRelative
    alphaBoolBlend {blendOption=Linear}
    boundBone {boneName=dyn_R_armguard_tie_2_jnt}
    componentAppliedLinearAccClamp=<3000 3000 300>
    gravityOverride {z=-980}
    physicsBodyDefinitions [{
      boundBone {boneName=dyn_R_armguard_tie_2_jnt}
      boxExtents=<5 5 5>
      constraintSetup {angularLimitsMax=<10 60 5>, angularLimitsMin=<-10 -60 -100>, angularTarget=<1 0 0>}
      localJointOffset=<1 0 0>
     }]
   }
   Output Pose(Result=Component To Local(ComponentPose=@n32)) {layerGroup=None}
  ]}
 EventGraph {nodes [
   Event OnBeginJumping()
   > Set jumpFalling(jumpFalling:double=1) -> (Output_Get:double)
   > Set jumpFallingBool(jumpFallingBool:bool=true) -> (Output_Get:bool)
   > Delay(Duration=0.75)
   > Set jumpFalling(jumpFalling=0)
   > Set jumpFallingBool(jumpFallingBool=false)
   Event BlueprintUpdateAnimation()
   > n1 = Sequence!(execute) -> (then_0, then_1, then_2:Exec)
   > n2 = Branch(Condition=OR Boolean!(A=Get bIsAccelerating2D(), B=Get jumpFallingBool() -> (jumpFallingBool:bool), C:bool=Get bIsSkyDiving()) -> (ReturnValue))
   > Set isMoving(isMoving:float=1) -> (Output_Get:float)
   > Set isNotMoving(isNotMoving:double=0) -> (Output_Get:double)
   Set isNotMoving(execute=@n2.else, isNotMoving=1)
   > Set isMoving(isMoving=0)
   n3 = Get Anim_Inst() -> (Anim_Inst:Anim Instance Object Reference)
   Is Valid(exec=@n1.then_1, InputObject=@n3)
   > Branch(Condition=Get bIsInFrontEnd())
   > Set customIdlePlaying(customIdlePlaying:bool=Montage_IsPlaying(self=@n3, Montage=/BRCosmetics/Animation/Game/MainPlayer/Menu/BR/CustomSkins/_s23/MagicMeadowM/MagicMeadow_Male_Idle_Montage)) -> (Output_Get:bool)
   n4 = BreakTransform(InTransform=GetDeltaTransformFromRefPose(self=GetOwnerSkeletalMeshComponent(), BoneName=thigh_l, BaseName=pelvis))
   Set L_thighForward(execute=@n1.then_2, L_thighForward:double=MapRangeClamped(Value=BreakRotator(InRot=@n4.Rotation).Yaw, InRangeA=-20, InRangeB=-40)) -> (Output_Get:double)
   Event BlueprintInitializeAnimation()
   > n5 = Set Body_Mesh(Body_Mesh=GetOwnerSkeletalMeshComponent())
   > Set Anim_Inst(Anim_Inst:Anim Instance Object Reference=GetAnimInstance(self=@n5)) -> (Output_Get:Anim Instance Object Reference)
  ]}
]
variables [
 Anim_Inst {type=AnimInstance, default=null}
 Body_Mesh {type=SkeletalMeshComponent}
 customIdlePlaying {type=boolean}
 isMoving {type=number}
 isNotMoving {type=number}
 jumpFalling {type=number}
 jumpFallingBool {type=boolean}
 L_thighForward {type=number}
 Skel_Mesh {type=SkeletalMeshComponent}
]
