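# Physics asset for the MagicMeadow medium male cosmetic's face accessory, custom idle variant, simulating its hair chains and ponytail with capsule bodies.
# Center, left and right chains hang from master joints on the head with angular limits, and the ponytail branches from the center master.
# Spine, clavicle, neck and head capsules form the body, joined by free-angle constraints.
# Extra free-angle constraints cross-link both side-chain tips to the second center joint.
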PhysicsAsset /BRCosmetics/Characters/Player/Male/Medium/Bodies/M_MED_MagicMeadow/Meshes/Parts/M_MED_MagicMeadow_FaceAcc_Physics_customIdle
bodies {
 clavicle_l [{shapeType=Capsule, length=12, radius=10, shapeName=clavicle_l_capsule, center=<8.9996 0.2335 -0.3721>, rotation=<rot 77.9147 4.9999 -0.0001>}]
 clavicle_r [{shapeType=Capsule, length=12, radius=8.4367, shapeName=clavicle_l_capsule, center=<-9.1 0 0>, rotation=<rot 80 0 0>}]
 dyn_C_hairchain_1_jnt [{shapeType=Capsule, length=7, radius=2.5, shapeName=dyn_C_hairchain_1_jnt_capsule, center=<2 0 0>, rotation=<rot 90 0 0>}]
 dyn_C_hairchain_2_jnt [{shapeType=Capsule, length=10, radius=6.1, shapeName=dyn_C_hairchain_2_jnt_capsule, center=<2 0 4.0513>, rotation=<rot 0 0 90>}]
 dyn_C_hairchain_3_jnt [{shapeType=Capsule, length=10, radius=6, shapeName=dyn_C_hairchain_3_jnt_capsule, center=<2 0 4.3171>, rotation=<rot 0 0 90>}]
 dyn_C_hairchain_4_jnt [{shapeType=Capsule, length=5, radius=6, shapeName=dyn_C_hairchain_4_jnt_capsule, center=<3 0 4.3171>, rotation=<rot 0 0 90>}]
 dyn_L_hairchain_1_jnt [{shapeType=Capsule, length=3, radius=2, shapeName=dyn_L_hairchain_1_jnt_capsule, center=<0 0.9394 -3.4969>}]
 dyn_L_hairchain_2_jnt [
  {shapeType=Capsule, length=3, radius=2, shapeName=dyn_L_hairchain_2_jnt_capsule, center=<0 0.6003 0>}
  {
   shapeType=Capsule
   length=3
   radius=1.3
   shapeName=dyn_L_hairchain_2_jnt_capsule
   center=<3.1027 -1.4946 0.6781>
   rotation=<rot -44.3057 -4.76 -22.7152>
  }
  {
   shapeType=Capsule
   length=3
   radius=1.3
   shapeName=dyn_L_hairchain_2_jnt_capsule
   center=<2.2472 0.5367 -3.5717>
   rotation=<rot 46.4571 -0.0079 -0.0085>
  }
  {
   shapeType=Capsule
   length=3
   radius=2
   shapeName=dyn_L_hairchain_2_jnt_capsule
   center=<1.1439 -0.4406 -7.8949>
   rotation=<rot 22.7852 -42.6354 3.7717>
  }
  {
   shapeType=Capsule
   length=3.4152
   radius=1.3252
   shapeName=dyn_L_hairchain_2_jnt_capsule
   center=<0.4802 -1.5249 -10.2606>
   rotation=<rot 10.2513 0 27.7226>
  }
 ]
 dyn_R_hairchain_1_jnt [{shapeType=Capsule, length=1.6108, radius=2, shapeName=dyn_R_hairchain_1_jnt_capsule, center=<3 0 0>, rotation=<rot 90 0 0>}]
 dyn_R_hairchain_2_jnt [
  {shapeType=Capsule, length=4.4163, radius=1.2304, shapeName=dyn_R_hairchain_1_jnt_capsule, center=<3 0 0>}
  {
   shapeType=Capsule
   length=2.591
   radius=1.2304
   shapeName=dyn_R_hairchain_2_jnt_capsule
   center=<4.3452 2.7577 4.9733>
   rotation=<rot -25.426 0 43.9826>
  }
  {shapeType=Capsule, length=3, radius=1, shapeName=dyn_R_hairchain_2_jnt_capsule, center=<2.6958 0 -5.75>, rotation=<rot 0 0 -19.6559>}
 ]
 dyn_ponytail_1_jnt [{shapeType=Capsule, length=3, radius=2, shapeName=dyn_ponytail_1_jnt_capsule, center=<2 0 4>, rotation=<rot 0 0 90>}]
 dyn_ponytail_2_jnt [{shapeType=Capsule, length=3, radius=2, shapeName=dyn_ponytail_2_jnt_capsule, center=<2 0 7>, rotation=<rot 0 0 90>}]
 dyn_ponytail_3_jnt [{shapeType=Capsule, length=3, radius=2, shapeName=dyn_ponytail_3_jnt_capsule, center=<2 0 8.3>, rotation=<rot 0 0 90>}]
 head [{shapeType=Capsule, length=1, radius=1, shapeName=head_capsule}]
 master_C_hairchain_jnt [{shapeType=Capsule, length=7, radius=2.5, shapeName=dyn_C_hairchain_1_jnt_capsule, center=<2 0 0>, rotation=<rot 90 0 0>}]
 master_L_hairchain_jnt [{shapeType=Capsule, length=3, radius=2, shapeName=master_L_hairchain_jnt_capsule, center=<1 1.4421 0>}]
 master_R_hairchain_jnt [{shapeType=Capsule, length=3, radius=2, shapeName=master_R_hairchain_jnt_capsule, center=<3 0 0>, rotation=<rot 90 0 0>}]
 neck_02 [
  {shapeType=Capsule, length=10, radius=5.1, shapeName=neck_01_capsule, center=<0 -3.4474 -0.4893>, rotation=<rot 90 0 0>}
  {shapeType=Capsule, length=4, radius=7.6, shapeName=neck_02_capsule, center=<-1.9885 -3 0>, rotation=<rot 90 0 0>}
 ]
 spine_05 [
  {shapeType=Capsule, length=15.805, radius=12, shapeName=spine_05_capsule, center=<0.9236 -1.3477 0>, rotation=<rot 0 -90 0>}
  {shapeType=Capsule, length=9.2827, radius=10, shapeName=spine_05_capsule, center=<5.15 -2.163 0>}
 ]
}
constraints [
 {bone1Name=dyn_C_hairchain_1_jnt, bone2Name=master_C_hairchain_jnt, swing1LimitDegrees=10, swing2LimitDegrees=60, twistLimitDegrees=0}
 {bone1Name=dyn_C_hairchain_2_jnt, bone2Name=dyn_C_hairchain_1_jnt, swing1LimitDegrees=10, swing2LimitDegrees=60, twistLimitDegrees=0}
 {bone1Name=dyn_C_hairchain_3_jnt, bone2Name=dyn_C_hairchain_2_jnt, swing1LimitDegrees=10, swing2LimitDegrees=60, twistLimitDegrees=0}
 {bone1Name=dyn_C_hairchain_4_jnt, bone2Name=dyn_C_hairchain_3_jnt, swing1LimitDegrees=10, swing2LimitDegrees=60, twistLimitDegrees=0}
 {
  bone1Name=clavicle_l
  bone2Name=spine_05
  swing1LimitDegrees=45
  swing1Motion=Free
  swing2LimitDegrees=45
  swing2Motion=Free
  twistLimitDegrees=45
  twistMotion=Free
 }
 {
  bone1Name=clavicle_r
  bone2Name=spine_05
  swing1LimitDegrees=45
  swing1Motion=Free
  swing2LimitDegrees=45
  swing2Motion=Free
  twistLimitDegrees=45
  twistMotion=Free
 }
 {
  bone1Name=neck_02
  bone2Name=spine_05
  swing1LimitDegrees=45
  swing1Motion=Free
  swing2LimitDegrees=45
  swing2Motion=Free
  twistLimitDegrees=45
  twistMotion=Free
 }
 {
  bone1Name=head
  bone2Name=neck_02
  swing1LimitDegrees=45
  swing1Motion=Free
  swing2LimitDegrees=45
  swing2Motion=Free
  twistLimitDegrees=45
  twistMotion=Free
 }
 {bone1Name=master_C_hairchain_jnt, bone2Name=head, swing1LimitDegrees=10, swing2LimitDegrees=40, twistLimitDegrees=0}
 {bone1Name=dyn_ponytail_1_jnt, bone2Name=master_C_hairchain_jnt, swing1LimitDegrees=8, swing2LimitDegrees=20, twistLimitDegrees=5}
 {bone1Name=dyn_ponytail_2_jnt, bone2Name=dyn_ponytail_1_jnt, swing1LimitDegrees=8, swing2LimitDegrees=20, twistLimitDegrees=5}
 {
  bone1Name=dyn_ponytail_3_jnt
  bone2Name=dyn_ponytail_2_jnt
  swing1LimitDegrees=8
  swing1Motion=Locked
  swing2LimitDegrees=30
  twistLimitDegrees=5
 }
 {bone1Name=master_L_hairchain_jnt, bone2Name=head, swing1LimitDegrees=30, swing2LimitDegrees=15, twistLimitDegrees=20}
 {bone1Name=dyn_L_hairchain_1_jnt, bone2Name=master_L_hairchain_jnt, swing1LimitDegrees=30, swing2LimitDegrees=15, twistLimitDegrees=20}
 {bone1Name=dyn_L_hairchain_2_jnt, bone2Name=dyn_L_hairchain_1_jnt, swing1LimitDegrees=30, swing2LimitDegrees=15, twistLimitDegrees=20}
 {bone1Name=master_R_hairchain_jnt, bone2Name=head, swing1LimitDegrees=30, swing2LimitDegrees=15, twistLimitDegrees=20}
 {bone1Name=dyn_R_hairchain_1_jnt, bone2Name=master_R_hairchain_jnt, swing1LimitDegrees=45, swing2LimitDegrees=45, twistLimitDegrees=45}
 {bone1Name=dyn_R_hairchain_2_jnt, bone2Name=dyn_R_hairchain_1_jnt, swing1LimitDegrees=45, swing2LimitDegrees=45, twistLimitDegrees=45}
 {
  bone1Name=dyn_L_hairchain_2_jnt
  bone2Name=dyn_C_hairchain_2_jnt
  swing1LimitDegrees=45
  swing1Motion=Free
  swing2LimitDegrees=45
  swing2Motion=Free
  twistLimitDegrees=45
  twistMotion=Free
 }
 {
  bone1Name=dyn_R_hairchain_2_jnt
  bone2Name=dyn_C_hairchain_2_jnt
  swing1LimitDegrees=45
  swing1Motion=Free
  swing2LimitDegrees=45
  swing2Motion=Free
  twistLimitDegrees=45
  twistMotion=Free
 }
]
