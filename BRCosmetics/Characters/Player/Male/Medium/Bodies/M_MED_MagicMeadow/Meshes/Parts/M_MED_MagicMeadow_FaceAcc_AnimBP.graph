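# Post-process-style AnimBP for the MagicMeadow male FaceAcc part, adding secondary motion to hair strands and chains.
# Blended parent and body poses feed spring AnimDynamics strands, then two RigidBody nodes switched by whether the front-end custom idle montage plays.
# Weighted constraints then pull hair-chain bones toward spine, neck and shoulders.
# The event graph adjusts gravity and constraint alphas for jumping or falling, and some alphas for the custom idle.
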AnimBlueprint /BRCosmetics/Characters/Player/Male/Medium/Bodies/M_MED_MagicMeadow/Meshes/Parts/M_MED_MagicMeadow_FaceAcc_AnimBP
target_skeleton=&/M_MED_MagicMeadow_FaceAcc_Skeleton
graphs [
 AnimGraph {nodes [
   n1 = Layered blend per bone(
    BasePose=Copy Pose From Mesh() {bCopyCurves=true, bUseAttachedParent=true}
    BlendPoses_0=Copy Pose From Mesh(SourceMeshComponent=Get Body_Mesh())
   ) {layerSetup [{branchFilters [
       {boneName=upperarm_l}
       {boneName=upperarm_r}
       {boneName=clavicle_l}
       {boneName=clavicle_r}
       {boneName=upperarm_twist_01_l}
       {boneName=upperarm_twist_01_r}
       {boneName=deform_upperarm_up_l}
       {boneName=deform_upperarm_up_r}
      ]}]}
   n2 = Anim Dynamics - Bone: dyn_C_hairstrand_1_jnt(ComponentPose=Local To Component(LocalPose=@n1), Alpha=1, ExternalForce="0,0,0") {
    angularDampingOverride=1
    angularSpringConstant=200
    bAngularSpring=true
    bOverrideAngularDamping=true
    bUseGravityOverride=true
    simulationSpace=RootRelative
    alphaBoolBlend {blendOption=Linear}
    boundBone {boneName=dyn_C_hairstrand_1_jnt}
    gravityOverride {z=-980}
    physicsBodyDefinitions [{
      boundBone {boneName=dyn_C_hairstrand_1_jnt}
      boxExtents=<5 5 5>
      constraintSetup {angularLimitsMax=<0 5 45>, angularLimitsMin=<0 -5 5>, angularTarget=<1 0 0>}
      localJointOffset=<1 0 0>
     }]
   }
   n3 = Anim Dynamics - Bone: dyn_C_hairstrand_2_jnt(ComponentPose=@n2, Alpha=1, ExternalForce="0,0,0") {
    angularDampingOverride=1
    angularSpringConstant=200
    bAngularSpring=true
    bOverrideAngularDamping=true
    bUseGravityOverride=true
    simulationSpace=RootRelative
    alphaBoolBlend {blendOption=Linear}
    boundBone {boneName=dyn_C_hairstrand_2_jnt}
    gravityOverride {z=-980}
    physicsBodyDefinitions [{
      boundBone {boneName=dyn_C_hairstrand_2_jnt}
      boxExtents=<5 5 5>
      constraintSetup {angularLimitsMax=<0 5 45>, angularLimitsMin=<0 -5 5>, angularTarget=<1 0 0>}
      localJointOffset=<1 0 0>
     }]
   }
   n4 = Anim Dynamics - Bone: dyn_L_hairstrand_1_1_jnt(ComponentPose=@n3, Alpha=1, ExternalForce="0,0,0") {
    angularDampingOverride=1
    angularSpringConstant=200
    bAngularSpring=true
    bOverrideAngularDamping=true
    bUseGravityOverride=true
    simulationSpace=RootRelative
    alphaBoolBlend {blendOption=Linear}
    boundBone {boneName=dyn_L_hairstrand_1_1_jnt}
    gravityOverride {z=-980}
    physicsBodyDefinitions [{
      boundBone {boneName=dyn_L_hairstrand_1_1_jnt}
      boxExtents=<5.5 5.5 5.5>
      constraintSetup {angularLimitsMax=<0 -5 -5>, angularLimitsMin=<0 -20 -30>, angularTarget=<1 0 0>}
      localJointOffset=<1 0 0>
     }]
   }
   n5 = Anim Dynamics - Bone: dyn_L_hairstrand_1_2_jnt(ComponentPose=@n4, Alpha=1, ExternalForce="0,0,0") {
    angularDampingOverride=1
    angularSpringConstant=200
    bAngularSpring=true
    bOverrideAngularDamping=true
    bUseGravityOverride=true
    simulationSpace=RootRelative
    alphaBoolBlend {blendOption=Linear}
    boundBone {boneName=dyn_L_hairstrand_1_2_jnt}
    gravityOverride {z=-980}
    physicsBodyDefinitions [{
      boundBone {boneName=dyn_L_hairstrand_1_2_jnt}
      boxExtents=<5.5 5.5 5.5>
      constraintSetup {angularLimitsMax=<0 -5 -5>, angularLimitsMin=<0 -20 -30>, angularTarget=<1 0 0>}
      localJointOffset=<1 0 0>
     }]
   }
   n6 = Anim Dynamics - Bone: dyn_L_hairstrand_1_3_jnt(ComponentPose=@n5, Alpha=1, ExternalForce="0,0,0") {
    angularDampingOverride=1
    angularSpringConstant=200
    bAngularSpring=true
    bOverrideAngularDamping=true
    bUseGravityOverride=true
    simulationSpace=RootRelative
    alphaBoolBlend {blendOption=Linear}
    boundBone {boneName=dyn_L_hairstrand_1_3_jnt}
    gravityOverride {z=-980}
    physicsBodyDefinitions [{
      boundBone {boneName=dyn_L_hairstrand_1_3_jnt}
      boxExtents=<5.5 5.5 5.5>
      constraintSetup {angularLimitsMax=<0 -5 -5>, angularLimitsMin=<0 -20 -30>, angularTarget=<1 0 0>}
      localJointOffset=<1 0 0>
     }]
   }
   n7 = Anim Dynamics - Bone: dyn_R_hairstrand_1_jnt(ComponentPose=@n6, Alpha=1, ExternalForce="0,0,0") {
    angularDampingOverride=1
    angularSpringConstant=100
    bAngularSpring=true
    bOverrideAngularDamping=true
    bUseGravityOverride=true
    simulationSpace=RootRelative
    alphaBoolBlend {blendOption=Linear}
    boundBone {boneName=dyn_R_hairstrand_1_jnt}
    gravityOverride {z=-980}
    physicsBodyDefinitions [{
      boundBone {boneName=dyn_R_hairstrand_1_jnt}
      boxExtents=<5.5 5.5 5.5>
      constraintSetup {angularLimitsMax=<0 20 -6>, angularLimitsMin=<0 -20 -30>, angularTarget=<1 0 0>}
      localJointOffset=<1 0 0>
     }]
   }
   n8 = Anim Dynamics - Bone: dyn_R_hairstrand_2_jnt(ComponentPose=@n7, Alpha=1, ExternalForce="0,0,0") {
    angularDampingOverride=1
    angularSpringConstant=100
    bAngularSpring=true
    bOverrideAngularDamping=true
    bUseGravityOverride=true
    simulationSpace=RootRelative
    alphaBoolBlend {blendOption=Linear}
    boundBone {boneName=dyn_R_hairstrand_2_jnt}
    gravityOverride {z=-980}
    physicsBodyDefinitions [{
      boundBone {boneName=dyn_R_hairstrand_2_jnt}
      boxExtents=<5.5 5.5 5.5>
      constraintSetup {angularLimitsMax=<0 20 -6>, angularLimitsMin=<0 -20 -30>, angularTarget=<1 0 0>}
      localJointOffset=<1 0 0>
     }]
   }
   n9 = Anim Dynamics - Bone: dyn_R_hairstrand_3_jnt(ComponentPose=@n8, Alpha=1, ExternalForce="0,0,0") {
    angularDampingOverride=1
    angularSpringConstant=100
    bAngularSpring=true
    bOverrideAngularDamping=true
    bUseGravityOverride=true
    simulationSpace=RootRelative
    alphaBoolBlend {blendOption=Linear}
    boundBone {boneName=dyn_R_hairstrand_3_jnt}
    gravityOverride {z=-980}
    physicsBodyDefinitions [{
      boundBone {boneName=dyn_R_hairstrand_3_jnt}
      boxExtents=<5.5 5.5 5.5>
      constraintSetup {angularLimitsMax=<0 20 -6>, angularLimitsMin=<0 -20 -30>, angularTarget=<1 0 0>}
      localJointOffset=<1 0 0>
     }]
   }
   n10 = Get customIdlePlaying() -> (customIdlePlaying:bool)
   n11 = RigidBody!(ComponentPose=@n9, bAlphaBoolEnabled:bool=@n10, OverrideWorldGravity:Vector="0,0,-980", ExternalForce) -> (Pose) {
    alphaInputType=Bool
    bClampLinearTranslationLimitToRefPose=true
    bOverrideWorldGravity=true
    overridePhysicsAsset=&/M_MED_MagicMeadow_FaceAcc_Physics_customIdle
    simulationSpace=BaseBoneSpace
    alphaBoolBlend {blendOption=Linear}
    baseBoneRef {boneName=spine_05}
    componentLinearAccScale=<0.24 0.24 0.24>
    componentLinearVelScale=<0.3 0.3 0.3>
    overrideWorldGravity {z=-980}
    simSpaceSettings {maxAngularAcceleration=1000, maxAngularVelocity=3, maxLinearAcceleration=250, maxLinearVelocity=500, velocityScaleZ=0, worldAlpha=1}
   }
   n12 = RigidBody(ComponentPose=@n11, bAlphaBoolEnabled=NOT Boolean(A=@n10), OverrideWorldGravity=Get gravityOverride() -> (gravityOverride:Vector)) {
    alphaInputType=Bool
    bClampLinearTranslationLimitToRefPose=true
    bOverrideWorldGravity=true
    overridePhysicsAsset=&/M_MED_MagicMeadow_FaceAcc_Physics
    simulationSpace=BaseBoneSpace
    alphaBoolBlend {blendOption=Linear}
    baseBoneRef {boneName=spine_05}
    componentLinearAccScale=<0.24 0.24 0.24>
    componentLinearVelScale=<0.3 0.3 0.3>
    overrideWorldGravity {z=-980}
    simSpaceSettings {maxAngularAcceleration=1000, maxAngularVelocity=3, maxLinearAcceleration=250, maxLinearVelocity=500, velocityScaleZ=0, worldAlpha=1}
   }
   n13 = "Constraint - dyn_L_hairchain_1_jnt (PPPPPPPP)"!(
    ComponentPose=@n12
    Alpha="Get ,3float"() -> (",3float":double)
    ConstraintWeights_0=0.1048
    ConstraintWeights_1:float=0.4063
    ConstraintWeights_2:float=0.0081
    ConstraintWeights_3:float=0.0443
    ConstraintWeights_4:float=0.0142
    ConstraintWeights_5:float=0.0329
    ConstraintWeights_6:float=0.3816
    ConstraintWeights_7:float=0.0078
   ) -> (Pose) {
    alphaBoolBlend {blendOption=Linear}
    alphaScaleBiasClamp {bInterpResult=true, interpSpeedDecreasing=20, interpSpeedIncreasing=20}
    boneToModify {boneName=dyn_L_hairchain_1_jnt}
    constraintSetup [
     {targetBone {boneName=spine_05}}
     {targetBone {boneName=clavicle_l}}
     {targetBone {boneName=upperarm_twist_01_l}}
     {targetBone {boneName=deform_upperarm_up_l}}
     {targetBone {boneName=spine_03}}
     {targetBone {boneName=spine_04}}
     {targetBone {boneName=neck_01}}
     {targetBone {boneName=neck_02}}
    ]
    constraintWeights [1,1,1,1,1,1,1,1]
   }
   n14 = "Constraint - dyn_L_hairchain_2_jnt (PPPPPPPP)"!(
    ComponentPose=@n13
    Alpha="Get ,4float"() -> (",4float":double)
    ConstraintWeights_0=0.1048
    ConstraintWeights_1:float=0.4063
    ConstraintWeights_2:float=0.0081
    ConstraintWeights_3:float=0.0443
    ConstraintWeights_4:float=0.0142
    ConstraintWeights_5:float=0.0329
    ConstraintWeights_6:float=0.3816
    ConstraintWeights_7:float=0.0078
   ) -> (Pose) {
    alphaBoolBlend {blendOption=Linear}
    alphaScaleBiasClamp {bInterpResult=true, interpSpeedDecreasing=20, interpSpeedIncreasing=20}
    boneToModify {boneName=dyn_L_hairchain_2_jnt}
    constraintSetup [
     {targetBone {boneName=spine_05}}
     {targetBone {boneName=clavicle_l}}
     {targetBone {boneName=upperarm_twist_01_l}}
     {targetBone {boneName=deform_upperarm_up_l}}
     {targetBone {boneName=spine_03}}
     {targetBone {boneName=spine_04}}
     {targetBone {boneName=neck_01}}
     {targetBone {boneName=neck_02}}
    ]
    constraintWeights [1,1,1,1,1,1,1,1]
   }
   n15 = "Constraint - dyn_R_hairchain_1_jnt (PPPPPPPP)"!(
    ComponentPose=@n14
    Alpha="Get ,3float"()
    ConstraintWeights_0=0.079
    ConstraintWeights_1:float=0.3287
    ConstraintWeights_2:float=0.0394
    ConstraintWeights_3:float=0.0544
    ConstraintWeights_4:float=0.0286
    ConstraintWeights_5:float=0.037
    ConstraintWeights_6:float=0.4031
    ConstraintWeights_7:float=0.0297
   ) -> (Pose) {
    alphaBoolBlend {blendOption=Linear}
    alphaScaleBiasClamp {bInterpResult=true, interpSpeedDecreasing=20, interpSpeedIncreasing=20}
    boneToModify {boneName=dyn_R_hairchain_1_jnt}
    constraintSetup [
     {targetBone {boneName=spine_05}}
     {targetBone {boneName=clavicle_r}}
     {targetBone {boneName=upperarm_twist_01_r}}
     {targetBone {boneName=deform_upperarm_up_r}}
     {targetBone {boneName=spine_03}}
     {targetBone {boneName=spine_04}}
     {targetBone {boneName=neck_01}}
     {targetBone {boneName=neck_02}}
    ]
    constraintWeights [1,1,1,1,1,1,1,1]
   }
   n16 = "Constraint - dyn_R_hairchain_2_jnt (PPPPPPPP)"!(
    ComponentPose=@n15
    Alpha="Get ,5float"() -> (",5float":double)
    ConstraintWeights_0=0.079
    ConstraintWeights_1:float=0.3287
    ConstraintWeights_2:float=0.0394
    ConstraintWeights_3:float=0.0544
    ConstraintWeights_4:float=0.0286
    ConstraintWeights_5:float=0.037
    ConstraintWeights_6:float=0.4031
    ConstraintWeights_7:float=0.0297
   ) -> (Pose) {
    alphaBoolBlend {blendOption=Linear}
    alphaScaleBiasClamp {bInterpResult=true, interpSpeedDecreasing=20, interpSpeedIncreasing=20}
    boneToModify {boneName=dyn_R_hairchain_2_jnt}
    constraintSetup [
     {targetBone {boneName=spine_05}}
     {targetBone {boneName=clavicle_r}}
     {targetBone {boneName=upperarm_twist_01_r}}
     {targetBone {boneName=deform_upperarm_up_r}}
     {targetBone {boneName=spine_03}}
     {targetBone {boneName=spine_04}}
     {targetBone {boneName=neck_01}}
     {targetBone {boneName=neck_02}}
    ]
    constraintWeights [1,1,1,1,1,1,1,1]
   }
   n17 = "Constraint - dyn_C_hairchain_2_jnt (P)"(ComponentPose=@n16, Alpha="Get ,5float"(), ConstraintWeights_0=1) {
    alphaBoolBlend {blendOption=Linear}
    boneToModify {boneName=dyn_C_hairchain_2_jnt}
    constraintSetup [{targetBone {boneName=spine_05}}]
    constraintWeights [1]
   }
   Output Pose(Result=Component To Local(ComponentPose=@n17)) {layerGroup=None}
  ]}
 EventGraph {nodes [
   Event BlueprintUpdateAnimation()
   > n1 = Sequence()
   n2 = Get Anim_Inst() -> (Anim_Inst:Anim Instance Object Reference)
   > Is Valid(InputObject=@n2)
   > Branch(Condition=Get bIsInFrontEnd())
   > Set customIdlePlaying(customIdlePlaying:bool=Montage_IsPlaying(self=@n2, Montage=/BRCosmetics/Animation/Game/MainPlayer/Menu/BR/CustomSkins/_s23/MagicMeadowM/MagicMeadow_Male_Idle_Montage)) -> (Output_Get:bool)
   n3 = Gravity Override State Function DEPRECATED(
    execute=@n1.then_1
    CharacterPartAnimInstance=Self-Reference()
    GameTimeInSeconds=GetGameTimeInSeconds()
    GravityOverrideParams=Get Gravity Override Params()
   )
   > n4 = Set gravityOverride(gravityOverride:Vector=@n3) -> (Output_Get:Vector)
   n5 = Get bIsFalling()
   n6 = Get bIsJumping()
   > Set gravityOverride(gravityOverride=SelectVector(A="0,0,-200", B=@n4, bPickA=OR Boolean(A=@n6, B=@n5)))
   n7 = OR Boolean(A=Get bIsJumping(), B=Get bIsFalling())
   > "Set ,3float"(",3float":double=SelectFloat(A=0, B=0.3, bPickA=@n7)) -> (Output_Get:double)
   > "Set ,4float"(",4float":double=SelectFloat(A=0, B=0.4, bPickA=@n7)) -> (Output_Get:double)
   > n8 = "Set ,85float"(",85float":double=SelectFloat(A=0, B=0.85, bPickA=@n7)) -> (Output_Get:double)
   > "Set ,5float"(",5float":double=SelectFloat(A=0.5, B=@n8, bPickA=Get customIdlePlaying())) -> (Output_Get:double)
   Event BlueprintInitializeAnimation()
   > n9 = Set Body_Mesh(Body_Mesh=GetOwnerSkeletalMeshComponent())
   > Set Anim_Inst(Anim_Inst:Anim Instance Object Reference=GetAnimInstance(self=@n9)) -> (Output_Get:Anim Instance Object Reference)
  ]}
]
variables [
 ",3float" {type=number}
 ",4float" {type=number}
 ",5float" {type=number}
 ",85float" {type=number}
 Anim_Inst {type=AnimInstance, default=null}
 Body_Mesh {type=SkeletalMeshComponent}
 customIdlePlaying {type=boolean}
 Gravity Override Params {type=GravityOverrideParamsStruct, default {onGround_Falling {gravityOverride=<0 0 -100>}}}
 gravityOverride {type=Vector}
 R upperarm forward {type=number}
 R upperarm up {type=number}
]
